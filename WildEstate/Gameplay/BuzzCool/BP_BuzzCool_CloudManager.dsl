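; Cloud manager Blueprint for BuzzCool that shows event-driven cloud stages by switching World Partition Data Layers.
; On the authority, it waits for match ready and initial calendar sync, then activates the mapped layer for each active calendar event.
; A name-to-layer map drives both loading and unloading, logging a missing manager or failed name lookup.
; Custom cheat events unload all mapped layers, then optionally activate a specific cloud or FT_Head layer.
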(event EventBeginPlay
  ;[1 1 1] Check for active calendar events, if active, show the current cloud stage.
  (Networking|SwitchHasAuthority
    (:Authority
      (Class|FortAsyncActionAuthorityMatchReady|MatchReadyAsync
        (:OnReady
          (MCPContext|InitialCalendarSyncCompleted
            (:OnSyncComplete
              (for _array_element (Variables|Default|GetCloudCalendarEventNames)
                (bind _biseventactive (CalendarEvents|GetCalendarEventInformation _array_element))
                (if _biseventactive
                  (CallFunction|LoadDataLayer _array_element))))))))))

(event Custom|BC_Cloud1
  ;[1 1 1] Cloud Cheats
  (CallFunction|UnloadBCDatalayers)
  (CallFunction|LoadDataLayer "BC_Cloud1"))

(event Custom|BC_Cloud2
  ;[1 1 1] Cloud Cheats
  (CallFunction|UnloadBCDatalayers)
  (CallFunction|LoadDataLayer "BC_Cloud2"))

(event Custom|BC_Cloud3
  ;[1 1 1] Cloud Cheats
  (CallFunction|UnloadBCDatalayers)
  (CallFunction|LoadDataLayer "BC_Cloud3"))

(event Custom|BC_CloudOff
  ;[1 1 1] Cloud Cheats
  (CallFunction|UnloadBCDatalayers))

(event Custom|FT_Show
  ;[1 1 1] Cloud Cheats
  (CallFunction|UnloadBCDatalayers)
  (CallFunction|LoadDataLayer "FT_Head"))

(fn LoadDataLayer
  (params (DataLayerToActivate :type String))
  (bind _cloudcalendareventdlmap (Variables|Default|GetCloudCalendarEventDLMap))
  (bind (_value _returnvalue) (Utilities|Map|Find _cloudcalendareventdlmap DataLayerToActivate))
  ;[1 1 1] Find Data Layer to Activate in Calendar Event Map and activate Data Layer if valid.
  (bind _returnvalue_2 (WorldPartition|GetDataLayerManager))
  (Utilities|IsValid _returnvalue_2
    (:Is%20Valid
      (if _returnvalue
        (DataLayers|SetDataLayerInstanceRuntimeState _returnvalue_2 (DataLayers|GetDataLayerInstancefromAsset _returnvalue_2 _value) "Activated" false)
        (else
          (Fortnite|FortShippingLog self (Utilities|String|Append "BC_BuzzCool_CloudManager: FAILED TO FIND DATA LAYER IN MAP WITH NAME:" DataLayerToActivate) true))))
    (:Is%20Not%20Valid
      (Fortnite|FortShippingLog self "BC_BuzzCool_CloudManager: FAILED TO FIND DATA LAYER MANAGER!" true))))

(fn UnloadBCDatalayers
  ;[1 1 1] Unload All Event Data Layers
  (bind _returnvalue (WorldPartition|GetDataLayerManager))
  (Utilities|IsValid _returnvalue
    (:Is%20Valid
      (bind (_keypin _valuepin) (Utilities|Map|ForEachLoop(Map) (Variables|Default|GetCloudCalendarEventDLMap)))
      (DataLayers|SetDataLayerInstanceRuntimeState _returnvalue (DataLayers|GetDataLayerInstancefromAsset _returnvalue _valuepin) "Unloaded" false))
    (:Is%20Not%20Valid
      (Fortnite|FortShippingLog self "BC_BuzzCool_CloudManager: FAILED TO FIND DATA LAYER MANAGER!" true))))

(fn ConstructionScript)
